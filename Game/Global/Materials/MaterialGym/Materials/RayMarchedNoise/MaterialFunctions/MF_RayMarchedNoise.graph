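# Material function that performs a single ray-march step through a noise signed distance field.
# It samples MF_RayMarchedNoiseSDF at the ray origin and subtracts that distance times the ray direction, as in sphere tracing.
# It outputs the new origin and unchanged direction, matching its inputs, which default to world position and normalized camera vector.
# An unconnected multiply by a constant remains in the graph.
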MaterialFunction /Game/Global/Materials/MaterialGym/Materials/RayMarchedNoise/MaterialFunctions/MF_RayMarchedNoise
graph {nodes [
  n1 = FunctionInput(Preview=WorldPosition().XYZ) {bUsePreviewValueAsDefault=true, inputName=ro}
  n2 = FunctionInput(Preview=Normalize(VectorInput=CameraVectorWS())) {bUsePreviewValueAsDefault=true, inputName=rd, sortPriority=1}
  n3 = Subtract(
   A=@n1
   B=Multiply(
    A=MaterialFunctionCall!(Vector=@n1) {materialFunction=&/MF_RayMarchedNoiseSDF}.Result
    B=@n2
   )
  )
  FunctionOutput(None=@n3) {outputName=ro}
  FunctionOutput(None=@n2) {outputName=rd, sortPriority=1}
  Multiply(A=@n3) {constB=0.5}
 ]}
