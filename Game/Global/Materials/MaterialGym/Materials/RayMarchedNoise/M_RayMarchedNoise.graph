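# Default Lit material that creates a two-color ray-marched noise effect with a Fresnel rim, enabled for skeletal meshes.
# Scaled pre-skinned position and normalized camera vector start eight unrolled MF_RayMarchedNoise calls, each later call taking the previous ro output for both inputs.
# Distance between the final and starting positions, times DistanceMultiplier, blends Color1 to Color2, plus a Color2-tinted Fresnel.
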Material /Game/Global/Materials/MaterialGym/Materials/RayMarchedNoise/M_RayMarchedNoise
graph {
 properties {bUsedWithMeshDeformer=true, bUsedWithSkeletalMesh=true, pixelDepthOffsetMode=PDOM_Legacy, usedShadingModels=Default Lit}
 nodes [
  n1 = VectorParameter() {parameterName=Color2, defaultValue=<rgba 0.9091 0.2668 1 1>}
  n2 = VertexInterpolator(VS=PreSkinnedPosition())
  n3 = ScalarParameter() {defaultValue=100, parameterName=UVDivisor}
  n4 = MaterialFunctionCall!(
   ro=NamedRerouteDeclaration(None=Divide(A=@n2.PS, B=@n3) {constB=100}) {name=LocalPosition, nodeColor=<rgba 0.3594 0 1 1>}
   rd=Normalize(VectorInput=CameraVectorWS())
  ) {materialFunction=/Game/Global/Materials/MaterialGym/Materials/RayMarchedNoise/MaterialFunctions/MF_RayMarchedNoise}
  n5 = MaterialFunctionCall(ro=@n4.ro, rd=@n4.ro) {materialFunction=&MF_RayMarchedNoise}
  n6 = MaterialFunctionCall(ro=@n5.ro, rd=@n5.ro) {materialFunction=&MF_RayMarchedNoise}
  n7 = MaterialFunctionCall(ro=@n6.ro, rd=@n6.ro) {materialFunction=&MF_RayMarchedNoise}
  n8 = MaterialFunctionCall(ro=@n7.ro, rd=@n7.ro) {materialFunction=&MF_RayMarchedNoise}
  n9 = MaterialFunctionCall(ro=@n8.ro, rd=@n8.ro) {materialFunction=&MF_RayMarchedNoise}
  n10 = MaterialFunctionCall(ro=@n9.ro, rd=@n9.ro) {materialFunction=&MF_RayMarchedNoise}
  n11 = Multiply(
   A=Distance(
    A=MaterialFunctionCall(ro=@n10.ro, rd=@n10.ro) {materialFunction=&MF_RayMarchedNoise}.ro
    B=NamedRerouteUsage()
   )
   B=ScalarParameter() {defaultValue=1, parameterName=DistanceMultiplier}
  )
  Add(
   A=Multiply(A=Fresnel() {baseReflectFraction=0, exponent=4}, B=@n1.RGB)
   B=LinearInterpolate(A=VectorParameter() {parameterName=Color1, defaultValue=<rgba 0.0185 0.0133 0.0312 1>}.RGB, B=@n1.RGB, Alpha=Saturate(None=@n11))
  )
 ]
}
